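# Niagara system for the idle effects of the Realm Male VinderTech glider: airfoil contrails, engine covers, engine meshes and flashing lights.
# Ribbons, sprites and lights follow skeletal mesh sockets; contrail per-unit spawning is gated on full deployment and forward dot.
# System timelines drive engine brightness and material parameters, with lights tracking the same blend.
# A front-end preview flag forces contrail spawning and full engine glow, while Blink toggles the flashing lights.
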NiagaraSystem /BRCosmetics/Gadgets/Assets/VinderTech_GliderChute/Glider_Realm_Male/FX/NS_Glider_Realm_Male_Idle
dependencies {
 usedDataInterfaces [&NiagaraDataInterfaceCurve, &NiagaraDataInterfaceSkeletalMesh]
 usedDynamicInputs [&Boolean_OrOperation]
 usedModules [
  &self:SystemSpawnScript.SystemScriptSource.SystemScriptGraph.NiagaraNodeAssignment_0.SetVariables_#1
  &SystemState
  &Timeline
 ]
}
emitters {
 AirFoil_Contrails {
  emitterUpdateScript [
   &EmitterState()
   &SpawnPerUnit(
    Module.Use Max Movement Threshold=false
    Module.Use Movement Tolerance=false
    Spawn Probability=1
    Spawn Spacing:NiagaraFloat=&Multiply_Float(
     A=1
     B=&MakeCustomFloatFromBool(
      Bool=&Boolean_AndOperation(
       Bool1=&Boolean_AndOperation(Bool1=$User.bIsFullyDeployed, Bool2=&SetBoolByFloatComparison(A=$User.ForwardDot, B=0.1))
       Bool2=true
      )
      True Float=1
      False Float=0
     )
    )
    Use Spawn Probability=true
   )
   &SpawnRate(
    SpawnRate=&MakeCustomFloatFromBool(Bool=$User.bisFrontEndPreview, True Float=60, False Float=0)
    Use Spawn Probability=false
   )
  ]
  particleSpawnScript [
   &FN_SetSpawnGroup(Inclusivity=true, Normalize Spawn Group=true, Number of Spawn Groups=2, Particles Per Group=15)
   &InitializeRibbon(Color=<rgba 1 1 1 0.125>, Lifetime=0.333, Ribbon Width=10, UsePositionOffset=false, Write Material Random=true)
   /SetVariables(
    Particles.RibbonID=/Niagara/DynamicInputs/NiagaraID/MakeNiagaraIDFromInt(Index:NiagaraInt32=$Particles.SpawnGroup, Acquire Tag:NiagaraInt32=0)
    Particles.RibbonLinkOrder:NiagaraFloat=$Particles.Normalized Spawn Group
   ) {isSetParameters=true}
   &SkeletalMeshLocation(
    Bone / Socket Index=&Add_Integer(A=$Particles.SpawnGroup, B=2)
    MaskSpawnGroup=false
    Mesh Sampling Type=::NewEnumerator5
    Skeletal Mesh=$System.Skeletal Mesh
    Socket Sampling Mode=::NewEnumerator5
   )
   &AddVelocity(Velocity=&MakeCustomVectorFromBool(Bool=$User.bisFrontEndPreview, True Vector=<0 1000 0>, False Vector=<0 0 0>))
  ]
  particleUpdateScript [
   &SolveForcesAndVelocity(Clamp Velocity=false, Limit Acceleration=false)
   &ParticleState()
   &ScaleColor(
    Scale Alpha=&FloatFromCurve(FloatCurve=DataInterface() {propertyValues {Curve {keys [<at 0 1>,<at 1 0>]}}})
    ScaleA=true
    ScaleRGB=true
   )
  ]
  rendererClasses [&NiagaraRibbonRendererProperties]
  renderers [{rendererClass=&NiagaraRibbonRendererProperties}]
 }
 Engine_Covers {
  emitterUpdateScript [
   &EmitterState(
    Inactive Response=::NewEnumerator1
    Life Cycle Mode=::NewEnumerator1
    Loop Behavior=::NewEnumerator1
    Loop Delay=&FN_FrontEndFloatModifier(Gameplay=0.1, Front End=0)
    Loop Duration Mode=::NewEnumerator1
    UseLoopDelay=true
   )
   &SpawnBurst_Instantaneous(Use Spawn Probability=false, UseLoopCountLimit=false)
   &SpawnBurst_Instantaneous(Spawn Group=1, Use Spawn Probability=false, UseLoopCountLimit=false) {module=SpawnBurst_Instantaneous001}
  ]
  particleSpawnScript [
   &InitializeParticle(
    Color=<rgba 0.275 0.5087 1.1 1>
    Lifetime=5
    Position Mode=::NewEnumerator0
    Sprite Size=<75 10>
    Sprite Size Mode=::NewEnumerator4
    Write Material Random=true
   )
   &SpriteFacingAndAlignment(Sprite Alignment=<0 0 1>, Sprite Facing=<0 1 0>, WriteAlignment=true, WriteFacing=true)
  ]
  particleUpdateScript [
   &ParticleState(Kill Particles When Lifetime Has Elapsed=false, Let Infinitely Lived Particles Die When Emitter Deactivates=true)
   &SkeletalMeshLocation(
    Bone / Socket Index=&SpawnGroup()
    MaskSpawnGroup=false
    Mesh Sampling Type=::NewEnumerator5
    Orientation Sampling=::NewEnumerator1
    Skeletal Mesh=$System.Skeletal Mesh
    Socket Sampling Mode=::NewEnumerator5
   )
   &ScaleColor(
    Scale Alpha=&Lerp_Float(
     A=0.0125
     B=1
     Alpha=&MakeCustomFloatFromBool(
      Bool=$User.bisFrontEndPreview
      True Float=1
      False Float=&ClampFloat(Float=&Add_Float(A=$System.Timeline001.NormalizedTime, B=$System.Timeline.NormalizedTime))
     )
    )
    Scale RGB=&VectorFromFloat(Value=&Lerp_Float(
      A=1
      B=1.5
      Alpha=&MakeCustomFloatFromBool(
       Bool=$User.bisFrontEndPreview
       True Float=1
       False Float=&ClampFloat(Float=&Add_Float(A=$System.Timeline001.NormalizedTime, B=$System.Timeline.NormalizedTime))
      )
     ))
    ScaleA=true
    ScaleRGB=true
   )
   /SetVariables(
    Particles.Light Position 2:NiagaraPosition=&Add_Vector(A=Linked() {linkedVariable {name=Particles.Position, type=NiagaraPosition}}, B=<-15 25 -1>)
    Particles.LightColor=&MultiplyLinearColor(
     Linear Color=<rgba 0.275 0.5087 1.1 1>
     Scale Factor=&MakeVector4_FromFloat(Float=&Multiply_Float(
       A=50
       B=&MakeCustomFloatFromBool(
        Bool=$User.bisFrontEndPreview
        True Float=10
        False Float=&ClampFloat(Float=&Add_Float(A=$System.Timeline001.NormalizedTime, B=$System.Timeline.NormalizedTime), Min=0.25)
       )
      ))
    )
    Particles.LightExponent=2
    Particles.LightPosition:NiagaraPosition=&Add_Vector(A=Linked() {linkedVariable {name=Particles.Position, type=NiagaraPosition}}, B=<15 25 -1>)
    Particles.LightRadius=25
   ) {isSetParameters=true}
  ]
  rendererClasses [&NiagaraSpriteRendererProperties, &NiagaraLightRendererProperties]
  renderers [
   {rendererClass=&NiagaraSpriteRendererProperties}
   {rendererClass=&NiagaraLightRendererProperties, rendererIndex=1}
   {rendererClass=&NiagaraLightRendererProperties, rendererIndex=2}
  ]
 }
 Engine_Ribbion_Meshes {
  emitterUpdateScript [
   &EmitterState(Life Cycle Mode=::NewEnumerator1, Loop Behavior=::NewEnumerator1, Loop Duration Mode=::NewEnumerator1, UseLoopDelay=false)
   &SpawnBurst_Instantaneous(Use Spawn Probability=false, UseLoopCountLimit=false)
  ]
  particleSpawnScript [&InitializeParticle(Color=<rgba 0.275 0.5087 1.1 1>, Lifetime=5, Mesh Scale Mode=::NewEnumerator3, UsePositionOffset=false, Write Material Random=true)]
  particleUpdateScript [
   &ParticleState(Kill Particles When Lifetime Has Elapsed=false, Let Infinitely Lived Particles Die When Emitter Deactivates=true)
   &ScaleColor(
    Scale RGB=&VectorFromFloat(Value=&Lerp_Float(
      A=1
      B=1.5
      Alpha=&MakeCustomFloatFromBool(
       Bool=$User.bisFrontEndPreview
       True Float=1
       False Float=&ClampFloat(Float=&Add_Float(A=$System.Timeline001.NormalizedTime, B=$System.Timeline.NormalizedTime))
      )
     ))
    ScaleA=true
    ScaleRGB=true
   )
   &DynamicMaterialParameters(
    Index 0 Param 1=&Lerp_Float(
     A=1
     B=-0.25
     Alpha=&MakeCustomFloatFromBool(
      Bool=$User.bisFrontEndPreview
      True Float=1.15
      False Float=&ClampFloat(Float=&Add_Float(A=$System.Timeline001.NormalizedTime, B=$System.Timeline.NormalizedTime))
     )
    )
    Index 0 Param 2=&FN_FrontEndFloatModifier(
     Gameplay=&Lerp_Float(
      A=0
      B=150
      Alpha=&ClampFloat(Float=&Add_Float(A=$System.Timeline001.NormalizedTime, B=$System.Timeline.NormalizedTime))
     )
     Front End=&Lerp_Float(A=300, B=75, Alpha=&ClampFloat(Float=$System.Timeline002.NormalizedTime))
    )
    Module.Index0Param0Write=true
    Module.Index0Param1Write=true
    Module.Index0Param2Write=false
    Module.Index0Param3Write=false
   )
  ]
  rendererClasses [&NiagaraMeshRendererProperties]
  renderers [{rendererClass=&NiagaraMeshRendererProperties, rendererIndex=0}]
 }
 Flashing_Lights {
  emitterUpdateScript [
   &EmitterState(
    Inactive Response=::NewEnumerator1
    Life Cycle Mode=::NewEnumerator1
    Loop Behavior=::NewEnumerator1
    Loop Delay=&FN_FrontEndFloatModifier(Gameplay=0.1, Front End=0)
    Loop Duration Mode=::NewEnumerator1
    UseLoopDelay=true
   )
   &SpawnBurst_Instantaneous(Spawn Count=6, Use Spawn Probability=false, UseLoopCountLimit=false)
  ]
  particleSpawnScript [
   &FN_SetSpawnGroup(Number of Spawn Groups=6)
   &InitializeParticle(
    Color=<rgba 2 1.0136 0 1>
    Sprite Rotation Mode=::NewEnumerator2
    Sprite Size Mode=::NewEnumerator3
    Uniform Sprite Size=20
    UsePositionOffset=false
    Write Material Random=true
   )
   &CameraOffset(Camera Offset Amount=5)
  ]
  particleUpdateScript [
   &ParticleState(Kill Particles When Lifetime Has Elapsed=false, Let Infinitely Lived Particles Die When Emitter Deactivates=true)
   &SkeletalMeshLocation(
    Bone / Socket Index=&Add_Integer(A=$Particles.SpawnGroup, B=2)
    MaskSpawnGroup=false
    Mesh Sampling Type=::NewEnumerator5
    Skeletal Mesh=$System.Skeletal Mesh
    Socket Sampling Mode=::NewEnumerator5
   )
   &ScaleColor(
    Scale Alpha=&MakeCustomFloatFromBool(Bool=$User.Blink, True Float=1, False Float=0)
    ScaleA=true
    ScaleRGB=true
   )
   &DynamicMaterialParameters(
    Index 0 Param 1=1
    Module.Param0WriteEnabled=true
    Module.Param1WriteEnabled=true
    Module.Param2WriteEnabled=true
    Module.Param3WriteEnabled=true
   )
  ]
  rendererClasses [&NiagaraSpriteRendererProperties]
  renderers [
   {rendererClass=&NiagaraSpriteRendererProperties}
   {rendererClass=&NiagaraSpriteRendererProperties, rendererIndex=1}
  ]
 }
}
summary {
 systemName=NS_Glider_Realm_Male_Idle
 emitters [
  {emitterName=AirFoil_Contrails, rendererClasses [&NiagaraRibbonRendererProperties]}
  {emitterName=Flashing_Lights, rendererClasses [&NiagaraSpriteRendererProperties]}
  {emitterName=Engine_Covers, rendererClasses [&NiagaraSpriteRendererProperties, &NiagaraLightRendererProperties]}
  {emitterName=Engine_Ribbion_Meshes, rendererClasses [&NiagaraMeshRendererProperties]}
 ]
 userVariables [
  User.bIsFrontEnd {type=NiagaraBool}
  User.bisFrontEndPreview {type=NiagaraBool}
  User.bIsFullyDeployed {type=NiagaraBool}
  User.Blink {type=NiagaraBool, defaultValue=true}
  User.ForwardDot {type=NiagaraFloat, defaultValue=0, description=""}
 ]
}
