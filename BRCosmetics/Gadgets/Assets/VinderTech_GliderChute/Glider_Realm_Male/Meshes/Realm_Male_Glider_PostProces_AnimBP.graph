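# Post-process Animation Blueprint for the Realm Male glider skeleton (VinderTech GliderChute) that resizes and repositions the glider on top of its incoming pose.
# The cached input pose runs through chained Modify Bone nodes in component space: Root is scaled down and offset, then both handle bones have their scale reset to one.
# A similar Frame bone node is left unconnected, and the event graph contains no logic.
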AnimBlueprint /BRCosmetics/Gadgets/Assets/VinderTech_GliderChute/Glider_Realm_Male/Meshes/Realm_Male_Glider_PostProces_AnimBP
target_skeleton=/Game/Gadgets/Assets/VinderTech_GliderChute/Glider_Realm_Male/Meshes/Realm_Male_Glider_Skeleton
graphs [
 AnimGraph {nodes [
   "Transform (Modify) Bone - Bone: Frame"(Alpha=1, Translation="0,-2,4", Rotation="0,0,0", Scale="0.75,0.75,0.75") {
    scaleMode=BMM_Replace
    translationMode=BMM_Replace
    alphaBoolBlend {blendOption=Linear}
    boneToModify {boneName=Frame}
    scale=<0.75 0.75 0.75>
    translation {y=-2, z=4}
   }
   n1 = "Transform (Modify) Bone - Bone: Root"(
    ComponentPose=Local To Component(LocalPose=Use cached pose 'InputCache'())
    Alpha=1
    Translation="0,-2,-3.5"
    Rotation="0,0,0"
    Scale="0.75,0.75,0.75"
   ) {
    scaleMode=BMM_Replace
    translationMode=BMM_Replace
    alphaBoolBlend {blendOption=Linear}
    boneToModify {boneName=Root}
    scale=<0.75 0.75 0.75>
    translation {y=-2, z=-3.5}
   }
   n2 = "Transform (Modify) Bone - Bone: JNT_Handle_R_01"(ComponentPose=@n1, Alpha=1, Translation="0,0,0", Rotation="0,0,0", Scale="1,1,1") {scaleMode=BMM_Replace, alphaBoolBlend {blendOption=Linear}, boneToModify {boneName=JNT_Handle_R_01}}
   Output Pose(Result=Component To Local(ComponentPose="Transform (Modify) Bone - Bone: JNT_Handle_L_01"(ComponentPose=@n2, Alpha=1, Translation="0,0,0", Rotation="0,0,0", Scale="1,1,1") {scaleMode=BMM_Replace, alphaBoolBlend {blendOption=Linear}, boneToModify {boneName=JNT_Handle_L_01}})) {layerGroup=None}
   Save cached pose 'InputCache'(Pose=Input Pose()) {cacheName=InputCache}
  ]}
 EventGraph {nodes [
   TryGetPawnOwner()
   Event BlueprintUpdateAnimation()
  ]}
]
variables []
